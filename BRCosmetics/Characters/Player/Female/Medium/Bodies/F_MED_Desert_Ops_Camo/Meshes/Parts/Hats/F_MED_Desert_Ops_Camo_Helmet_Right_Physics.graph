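# Physics asset for the Desert Ops Camo female medium helmet (Right variant), defining physics bodies for its dangling straps.
# A head capsule anchors two three-link chains of small capsules on the left and right dyn_strap bones.
# Each link is constrained to its parent with identical limits, narrow on swing1 and wide on swing2 and twist, so the straps swing mostly in one plane while being able to twist.
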PhysicsAsset /BRCosmetics/Characters/Player/Female/Medium/Bodies/F_MED_Desert_Ops_Camo/Meshes/Parts/Hats/F_MED_Desert_Ops_Camo_Helmet_Right_Physics
bodies {
 dyn_strap_01_l [{shapeType=Capsule, length=2.005, radius=1.005, center=<0 -0.0001 -1.553>}]
 dyn_strap_01_r [{shapeType=Capsule, length=3.25, radius=0.75, center=<0 0.0001 -1.9689>}]
 dyn_strap_02_l [{shapeType=Capsule, length=2.255, radius=0.755, center=<0 0 -1.886>}]
 dyn_strap_02_r [{shapeType=Capsule, length=1.755, radius=0.755, center=<0 0.0001 -1.9302>, rotation=<rot 0 -90 0>}]
 dyn_strap_03_l [{shapeType=Capsule, length=1.755, radius=1.005, center=<0 0 -2.1868>}]
 dyn_strap_03_r [{shapeType=Capsule, length=1.377, radius=1.1295, center=<0 0 -1.7694>}]
 head [{shapeType=Capsule, length=10.5257, radius=6.7046, rotation=<rot 90 0 0>}]
}
constraints [
 {bone1Name=dyn_strap_01_r, bone2Name=head, swing1LimitDegrees=7, swing2LimitDegrees=70, twistLimitDegrees=70}
 {bone1Name=dyn_strap_02_r, bone2Name=dyn_strap_01_r, swing1LimitDegrees=7, swing2LimitDegrees=70, twistLimitDegrees=70}
 {bone1Name=dyn_strap_03_r, bone2Name=dyn_strap_02_r, swing1LimitDegrees=7, swing2LimitDegrees=70, twistLimitDegrees=70}
 {bone1Name=dyn_strap_01_l, bone2Name=head, swing1LimitDegrees=7, swing2LimitDegrees=70, twistLimitDegrees=70}
 {bone1Name=dyn_strap_02_l, bone2Name=dyn_strap_01_l, swing1LimitDegrees=7, swing2LimitDegrees=70, twistLimitDegrees=70}
 {bone1Name=dyn_strap_03_l, bone2Name=dyn_strap_02_l, swing1LimitDegrees=7, swing2LimitDegrees=70, twistLimitDegrees=70}
]
