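# Animation Blueprint for the Desert Ops Camo female hat face accessory, with physics-driven straps.
# It layers head-mesh face attach and jaw bones onto the attached parent's pose, then optionally rotates the left strap bone.
# Chained left and right RigidBody nodes simulate on that cached pose, bypassed at higher LODs or when rigid bodies are disabled.
# Customization tags gate hat physics and swap left-side simulation for the strap offset by hair type.
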AnimBlueprint /BRCosmetics/Characters/Player/Female/Medium/Bodies/F_MED_Desert_Ops_Camo/Meshes/Parts/Hats/F_MED_Desert_Ops_Camo_Hat_FaceAcc_AnimBP
target_skeleton=/Game/Characters/Player/Female/Medium/Bodies/F_MED_Desert_Ops_Camo/Meshes/Parts/Hats/F_MED_Desert_Ops_Camo_Hat_FaceAcc_Skeleton
graphs [
 AnimGraph {nodes [
   n1 = RigidBody(ComponentPose=Local To Component(LocalPose=Use cached pose 'Start_Cache'()), Alpha=Get long_hair_alpha() -> (long_hair_alpha:double)) {
    bOverrideWorldGravity=true
    overridePhysicsAsset=&/F_MED_Desert_Ops_Camo_Helmet_Left_Physics
    alphaBoolBlend {blendOption=Linear}
    alphaScaleBiasClamp {scale=0.6}
    componentAppliedLinearAccClamp=<100 100 100>
    componentLinearAccScale {z=0.1}
    componentLinearVelScale=<10 10 10>
    overrideWorldGravity {z=-980}
   }
   n2 = RigidBody(ComponentPose=@n1, Alpha=0.6) {
    bOverrideWorldGravity=true
    overridePhysicsAsset=&/F_MED_Desert_Ops_Camo_Helmet_Right_Physics
    alphaBoolBlend {blendOption=Linear}
    componentAppliedLinearAccClamp=<100 100 100>
    componentLinearAccScale {z=0.1}
    componentLinearVelScale=<10 10 10>
    overrideWorldGravity {z=-980}
   }
   n3 = Component To Local(ComponentPose=@n2)
   n4 = Use cached pose 'Start_Cache'()
   n5 = Get hat_rigids_active() -> (hat_rigids_active:bool)
   Output Pose(Result=Blend Poses by bool(
     BlendPose_0=Blend Poses by bool(BlendPose_0=@n3, BlendPose_1=@n4, bActiveValue=@n5) {blendType=Linear}
     BlendPose_1=Use cached pose 'Start_Cache'()
     bActiveValue=Get rigids_active() -> (rigids_active:bool)
    ) {blendType=Linear}) {layerGroup=None}
   n6 = Copy Pose From Mesh(SourceMeshComponent=Get Head_Mesh() -> (Head_Mesh:Skeletal Mesh Component Object Reference)) {bCopyCurves=true}
   n7 = Copy Pose From Mesh() {bUseAttachedParent=true}
   n8 = "Transform (Modify) Bone - Bone: dyn_strap_01_l"(
    ComponentPose=Local To Component(LocalPose=Layered blend per bone(BasePose=@n7, BlendPoses_0=@n6) {layerSetup [{branchFilters [{boneName=faceAttach}, {boneName=jaw}]}]})
    Alpha=Get strap_mod_alpha() -> (strap_mod_alpha:double)
    Translation="0,0,0"
    Rotation="-15.000,180.000,-180.000"
    Scale="1,1,1"
   ) {
    rotationMode=BMM_Additive
    rotationSpace=BCS_BoneSpace
    alphaBoolBlend {blendOption=Linear}
    boneToModify {boneName=dyn_strap_01_l}
    rotation=<rot -15 180 -180>
   }
   Save cached pose 'Start_Cache'(Pose=Component To Local(ComponentPose=@n8)) {cacheName=Start_Cache}
  ]}
 EventGraph {nodes [
   Event BlueprintInitializeAnimation()
   > Set Body_Mesh(Body_Mesh=GetOwnerSkeletalMeshComponent())
   > n1 = Cast To PlayerPawn_Athena(Object=TryGetPawnOwner())
   > n2 = Set Pawn Owner(Pawn Owner:Player Pawn Athena Object Reference=@n1) -> (Output_Get:Player Pawn Athena Object Reference)
   > Set Head_Mesh(Head_Mesh:Skeletal Mesh Component Object Reference=GetSkeletalMeshForPartType(self=@n2, PartType=Head)) -> (Output_Get:Skeletal Mesh Component Object Reference)
   Event OnCharacterCustomizationLoaded()
   > n3 = Sequence()
   > n4 = Branch(Condition=CharacterItemDefinitionHasAnyMetaTag(Tags=Make Array("[0]"=(TagName="Cosmetics.MetaTags.Hat.01"))))
   > Set hat_rigids_active(hat_rigids_active:bool=true) -> (Output_Get:bool)
   Set hat_rigids_active(execute=@n4.else, hat_rigids_active=false)
   n5 = Branch(execute=@n3.then_1, Condition=CharacterItemDefinitionHasAnyMetaTag(Tags=Make Array("[0]"=(TagName="Cosmetics.MetaTags.Hair.00"))))
   > Set long_hair_active(long_hair_active:bool=true) -> (Output_Get:bool)
   > Set long_hair_alpha(long_hair_alpha:double=0) -> (Output_Get:double)
   > Set strap_mod_alpha(strap_mod_alpha:double=1) -> (Output_Get:double)
   Set long_hair_active(execute=@n5.else, long_hair_active=false)
   > Set long_hair_alpha(long_hair_alpha=1)
   > Set strap_mod_alpha(strap_mod_alpha=0)
   Event BlueprintUpdateAnimation()
   n6 = Get CurrentLOD()
   n7 = Get bRigidBodyEnabled()
   > n8 = Branch(Condition=OR Boolean(A=NOT Boolean(A=@n7), B="integer > integer"(A=@n6, B=1) {is=title K2Node_CallFunction}))
   > Set rigids_active(rigids_active:bool=false) -> (Output_Get:bool)
   Set rigids_active(execute=@n8.else, rigids_active=true)
   TryGetPawnOwner()
  ]}
]
variables [
 Body_Mesh {type=SkeletalMeshComponent}
 hat_rigids_active {type=boolean}
 Head_Mesh {type=SkeletalMeshComponent}
 long_hair_active {type=boolean}
 long_hair_alpha {type=number}
 Pawn Owner {type=PlayerPawn_Athena_C}
 rigids_active {type=boolean}
 strap_mod_alpha {type=number}
]
